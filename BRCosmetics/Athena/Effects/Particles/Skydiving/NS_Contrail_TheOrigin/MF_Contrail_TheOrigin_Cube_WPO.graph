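# Material function that computes noise-driven World Position Offset, named for the cube particles in TheOrigin skydiving contrail.
# A panning grayscale noise texture is sampled with object-stabilized screen-space UVs, with parameters for scale and speed.
# The noise is multiplied by Dynamic Parameter X replicated on all three axes, a negated, doubled Scale channel, and vertex color red.
# The dynamic Y and Z channels and a ParticlePositionWS node are unused.
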MaterialFunction /BRCosmetics/Athena/Effects/Particles/Skydiving/NS_Contrail_TheOrigin/MF_Contrail_TheOrigin_Cube_WPO
graph {nodes [
  n1 = DynamicParameter() {parameterIndex=1, defaultValue=<rgba 0 0 1 1>, paramNames [X,Y,Z,Scale]}
  n2 = ObjectPositionWS()
  n3 = ScalarParameter() {defaultValue=0.075, group=WorldPositionOffset, parameterName=WPO_Noise_Scale, sortPriority=1}
  n4 = ScalarParameter() {defaultValue=1, group=WorldPositionOffset, parameterName=WPO_Speed, sortPriority=1}
  n5 = TextureSampleParameter2D(UVs=Panner(
    Coordinate=MaterialFunctionCall!(Position=@n2, UVRadialScale=@n3) {materialFunction=/Game/Weapons/WeaponSkins/MaterialFunction/ObjectStabilizedScreenSpaceUVs}.StabilizedUVs
    Speed=MaterialFunctionCall(Y=@n4) {materialFunction=&MakeFloat2}.Result
   ) {speedX=1, speedY=1}) {
   group=Diffuse
   parameterName=WPO_Noise
   samplerSource=SSM_Wrap_WorldGroupSettings
   samplerType=SAMPLERTYPE_Grayscale
   sortPriority=1
   texture=&LowResBlurredNoise
  }
  n6 = Multiply(
   A=MaterialFunctionCall(X=@n1.X, Y=@n1.X, Z=@n1.X) {materialFunction=&MakeFloat3}.Result
   B=@n5.RGB
  )
  n7 = Multiply(A=@n1.Scale) {constB=-2}
  FunctionOutput(None=Multiply(A=Multiply(A=@n6, B=@n7), B=VertexColor().R))
  ParticlePositionWS()
 ]}
