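# Unlit Niagara mesh-particle and ribbon material for the cubes in The Origin skydiving contrail.
# Emissive adds panning inverted noise tinted by particle color, a gradient-masked glow scaled by dynamic parameter, and an inverted Fresnel on an offset normal.
# Further chains build opacity from particle alpha, masks, depth and near-camera fades, and blend in a cube WPO function by parameter.
# The graph lists no root connections, and several glow, specular and normal nodes feed nothing shown.
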Material /BRCosmetics/Athena/Effects/Particles/Skydiving/NS_Contrail_TheOrigin/M_Contrail_TheOrigin_Cubes
graph {
 properties {
  bUsedWithNiagaraMeshParticles=true
  bUsedWithNiagaraRibbons=true
  opacityMaskClipValue=0.16
  pixelDepthOffsetMode=PDOM_Legacy
  previewMesh=&/SM_SkyDives_TheOrigin_Cube
  shadingModel=MSM_Unlit
  translucencyLightingMode=TLM_VolumetricPerVertexDirectional
  usedShadingModels=Unlit
 }
 nodes [
  n1 = MaterialFunctionCall() {materialFunction=&LocalPosition}
  n2 = VectorParameter() {group=Emissive, parameterName=Glow Offset - Radius, defaultValue=<rgba 0 0 0 1>}
  n3 = Power(Base=Multiply(A=SphereMask(A=@n1."Instance Local Position (Excluding Offsets)", B=@n2.RGB) {attenuationRadius=10, hardnessPercent=0}) {constB=2})
  Add(A=@n3) {constB=0.25}
  AtmosphericLightVector()
  DynamicParameter() {parameterIndex=1, defaultValue=<rgba 0 0 1 1>, paramNames [X,Y,Z,Scale]}
  LinearInterpolate(Alpha=Power(
    Base=NamedRerouteDeclaration(None=@n3) {name=InnerGlow, nodeColor=<rgba 1 0 0.1641 1>}
    Exp=ScalarParameter() {defaultValue=2, group=Emissive, parameterName=Emissive Grad Power, sortPriority=1}
   ))
  LinearInterpolate(Alpha=Power(Base=NamedRerouteUsage(), Exp=ScalarParameter() {defaultValue=2, group=Diffuse, parameterName=Diff Grad Power, sortPriority=3})) {constA=0.25}
  LinearInterpolate(
   B=MaterialFunctionCall() {materialFunction=&/MF_Contrail_TheOrigin_Cube_WPO}.Result
   Alpha=ScalarParameter() {group=WorldPositionOffset, parameterName=Use WPO}
  )
  n4 = TextureSampleParameter2D() {
   group=Specular
   parameterName=SpecMasks
   samplerSource=SSM_Wrap_WorldGroupSettings
   samplerType=SAMPLERTYPE_Masks
   texture=/Game/Environments/Sets/ArtDeco/Architecture/Textures/T_AD_Glass_S
  }
  LinearInterpolate(
   A=ScalarParameter() {defaultValue=0.15, group=Specular, parameterName=SpecRoughnessMin, sortPriority=1}
   B=ScalarParameter() {defaultValue=1, group=Specular, parameterName=SpecRoughnessMax, sortPriority=2}
   Alpha=MaterialFunctionCall(Color=@n4.RGB) {materialFunction=&SplitComponents}.B
  )
  n5 = VectorParameter() {group=Emissive, parameterName="Emissive Color && Intensity", sortPriority=2, defaultValue=<rgba 1 1 1 0>}
  LinearInterpolate(A=@n5.RGB, B=Multiply(B=@n5.RGB), Alpha=ScalarParameter() {group=Emissive, parameterName=UseDiffuseinEmissive, sortPriority=1})
  MaterialFunctionCall() {materialFunction=&ObjectLocalBounds}
  n6 = ParticleColor()
  n7 = Panner(Coordinate=TextureCoordinate() {uTiling=0.3, vTiling=0.3}) {speedY=1}
  n8 = TextureSampleParameter2D(UVs=@n7) {
   group=Diffuse
   parameterName=Diffuse
   samplerSource=SSM_Wrap_WorldGroupSettings
   samplerType=SAMPLERTYPE_LinearColor
   sortPriority=1
   texture=/Game/Characters/Player/Male/Medium/Bodies/M_MED_SRDrift/Textures/T_ElectricNoise
  }
  n9 = Multiply(A=Add(A=OneMinus(None=@n8.G)) {constB=0.2}, B=ScalarParameter() {defaultValue=0.25, group=Diffuse, parameterName=Diff_Mult, sortPriority=3})
  n10 = DynamicParameter() {defaultValue=<rgba 1 1 0 1>, paramNames [Emissive Intensity, ColorChartUV, Erode, Param4]}
  n11 = TextureSampleParameter2D() {
   group=Emissive
   parameterName=Emissive Mask Texture
   samplerSource=SSM_Wrap_WorldGroupSettings
   sortPriority=3
   texture=/Game/Global/Textures/Default/Masks_and_Gradients/T_RadialGradient
  }
  n12 = ParticleColor()
  n13 = Multiply(A=Multiply(A=Multiply(A=@n5.RGB, B=@n10.Emissive Intensity), B=@n11.G), B=@n12.RGB)
  n14 = VertexNormalWS()
  n15 = Constant3Vector() {constant=<rgba 0 0 -0.1 0>}
  n16 = Fresnel(Normal=Normalize(VectorInput=Add(A=@n14, B=@n15))) {exponent=0.25}
  n17 = Power(Base=OneMinus(None=@n16)) {constExponent=3}
  MaterialFunctionCall(Emissive=Add(A=Multiply(A=@n6.RGB, B=@n9), B=Add(A=@n13, B=@n17))) {materialFunction=&MF_FX_EmissiveAdjustments}
  MaterialFunctionCall() {materialFunction=&ScreenAlignedUVs}
  Multiply() {constB=3}
  Multiply(A=LinearInterpolate(Alpha=@n16) {constA=1, constB=0}, B=@n11.G) {constA=1}
  n18 = ParticleColor()
  n19 = LinearInterpolate(B=@n8.A, Alpha=ScalarParameter() {defaultValue=1, group=Diffuse, parameterName=Diffuse Alpha Blend, sortPriority=3}) {constA=1}
  n20 = Multiply(
   A=TextureCoordinate()
   B=ComponentMask(None=VectorParameter() {group=Opacity, parameterName=Opacity Mask UV Scale, sortPriority=3, defaultValue=<rgba 1 1 0 0>}.RGB) {g=true, r=true}
  )
  n21 = TextureSampleParameter2D(UVs=LinearInterpolate(
    A=@n20
    B=MaterialFunctionCall(XY=@n20) {materialFunction=&Swizzle}.YX
    Alpha=ScalarParameter() {group=Opacity, parameterName=Swizzle Opacity Mask UVs, sortPriority=4}
   )) {
   group=Opacity
   parameterName=Opacity_Mask_Texture
   samplerSource=SSM_Wrap_WorldGroupSettings
   sortPriority=1
   texture=&T_WhitesRGB
  }
  Multiply(
   A=MaterialFunctionCall(Distance=ScalarParameter() {group=Opacity, parameterName=Near Camera Fade Distance}) {materialFunction=&Near_Camera_Fade}.Result
   B=Multiply(
    A=Multiply(A=@n18.A, B=@n19)
    B=ChannelMaskParameter(None=@n21.RGBA) {group=Opacity, parameterName=Opacity_Mask Channel Mask, sortPriority=2}.R
   )
  )
  ObjectBounds()
  n22 = MaterialFunctionCall() {materialFunction=&ObjectLocalBounds}
  n23 = Divide(A=Subtract(A=@n1.Instance Local Position, B=Multiply(A=@n2.RGB, B=@n22.Local Bounds Size)), B=Multiply(A=@n2.A, B=@n22.Local Bounds Size))
  Saturate(None=OneMinus(None=MaterialFunctionCall(Vector 3=@n23) {materialFunction=/Engine/Functions/Engine_MaterialFunctions02/Utility/VectorLength}.V3 Length))
  n24 = MaterialFunctionCall(
   Camera Vector=CameraVectorWS()
   Invert Fresnel=StaticBoolParameter() {defaultValue=true, parameterName=InvertFresnel}
   Power=ScalarParameter() {defaultValue=1, group=Emissive, parameterName=Emissive Mask Fresnel Exponent, sortPriority=5}
   Use Cheap contrast=StaticBoolParameter() {defaultValue=true, parameterName=UseCheapContrast}
  ) {materialFunction=&Fresnel_Function}
  n25 = DepthFade(Opacity=@n24.Result) {fadeDistanceDefault=15}
  n26 = LinearInterpolate(Alpha=@n24.Result) {constA=1, constB=0.25}
  Saturate(None=Multiply(A=@n12.A, B=Add(A=@n25, B=@n26)))
  ScalarParameter() {defaultValue=1, group=WorldPositionOffset, parameterName=WPO_LargeWaveScale, sortPriority=2}
  ScalarParameter() {defaultValue=0.5, group=WorldPositionOffset, parameterName=WPO_WaveScale, sortPriority=3}
  ScalarParameter() {defaultValue=0.33, group=Emissive, parameterName=Emissive Mask Fresnel Min Value, sortPriority=5}
  ScalarParameter() {defaultValue=1, group=Emissive, parameterName=Emissive Mask Fresnel Max Value, sortPriority=5}
  ScalarParameter() {group=Emissive, parameterName=Emissive Mask Fresnel Fake Slerp, sortPriority=5}
  TextureSampleParameter2D() {
   group=Normal
   parameterName=Normals
   samplerSource=SSM_Wrap_WorldGroupSettings
   samplerType=SAMPLERTYPE_Normal
   texture=/Engine/EngineMaterials/FlatNormal
  }
  TextureSample(UVs=@n7) {samplerType=SAMPLERTYPE_Masks, texture=/Game/Weapons/WeaponSkins/Wraps/Textures/Mash/Mash_Noise}
  TextureSample() {texture=/Game/Athena/Deimos/Spawners/RiftSpawners/Assets/CubeGrowth/T_CubeGrowth_occlusion}
 ]
}
