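# Unlit, masked, two-sided Niagara mesh-particle material for the background (BG) layer of the cubes in TheOrigin skydiving contrail.
# A dark purple Color parameter runs through the shared FX emissive adjustments function, alongside a contrail-specific cube world position offset function.
# A TwoSidedSign lerp yields zero on front faces and a positive value on back faces, apparently as a mask that leaves only the inside faces visible.
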Material /BRCosmetics/Athena/Effects/Particles/Skydiving/NS_Contrail_TheOrigin/M_Contrail_TheOrigin_Cubes_Main_BG
graph {
 properties {
  bCastRayTracedShadows=false
  bUsedWithNiagaraMeshParticles=true
  blendMode=BLEND_Masked
  pixelDepthOffsetMode=PDOM_Legacy
  shadingModel=MSM_Unlit
  twoSided=true
  usedShadingModels=Unlit
 }
 nodes [
  LinearInterpolate(Alpha=TwoSidedSign()) {constA=1, constB=0}
  MaterialFunctionCall(Emissive=VectorParameter() {parameterName=Color, defaultValue=<rgba 0.0516 0.0289 0.0729 1>}.RGB) {materialFunction=&MF_FX_EmissiveAdjustments}
  MaterialFunctionCall() {materialFunction=&/MF_Contrail_TheOrigin_Cube_WPO}
 ]
}
