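# Niagara skydiving contrail for The Origin cosmetic, trailing cube effects from two sampled skeletal mesh bones or sockets.
# A pulsing, squash-and-stretch main block with light and sprite pairs with energy meshes, while three cube emitters, one on GPU, spawn streaks and inner cubes after a short delay.
# System dive velocity drives orientation and offsets, an integrator scales intensity, and user ColorA and ColorB tint everything.
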NiagaraSystem /BRCosmetics/Athena/Effects/Particles/Skydiving/NS_Contrail_TheOrigin/NS_Contrail_TheOrigin
dependencies {
 usedDataInterfaces [
  &NiagaraDataInterfaceMeshRendererInfo
  &NiagaraDataInterfaceCurve
  &NiagaraDataInterfaceVectorCurve
  &NiagaraDataInterfaceSkeletalMesh
 ]
 usedDynamicInputs [
  &SetBoolByFloatComparison
  &NormalizeVector
  &FN_FrontEndVector3Modifier
  &MakeFloatFromVector
  &TransformVector
  &FN_FrontEndFloatModifier
 ]
 usedModules [
  &self:SystemSpawnScript.SystemScriptSource.SystemScriptGraph.NiagaraNodeAssignment_7.SetVariables_#1
  &SystemState
  &self:SystemSpawnScript.SystemScriptSource.SystemScriptGraph.NiagaraNodeAssignment_2.SetVariables_#2
  &FN_Integrator
  &self:SystemSpawnScript.SystemScriptSource.SystemScriptGraph.NiagaraNodeAssignment_6.SetVariables_#3
  &self:SystemSpawnScript.SystemScriptSource.SystemScriptGraph.NiagaraNodeAssignment_5.SetVariables_#4
  &self:SystemSpawnScript.SystemScriptSource.SystemScriptGraph.NiagaraNodeAssignment_4.SetVariables_#5
  &self:SystemSpawnScript.SystemScriptSource.SystemScriptGraph.NiagaraNodeAssignment_3.SetVariables_#6
  &self:SystemSpawnScript.SystemScriptSource.SystemScriptGraph.NiagaraNodeAssignment_0.SetVariables_#7
 ]
}
emitters {
 Block_Main_Light {
  emitterUpdateScript [
   &EmitterState(
    Inactive Response=::NewEnumerator1
    Life Cycle Mode=::NewEnumerator1
    Loop Behavior=::NewEnumerator1
    Loop Duration Mode=::NewEnumerator1
    UseLoopDelay=false
   ) {module=EmitterState001}
   &SpawnBurst_Instantaneous(Spawn Count=2, Use Spawn Probability=false, UseLoopCountLimit=false)
   /SetVariables(Emitter.WPO_Speed=1.2) {isSetParameters=true}
  ]
  particleSpawnScript [
   &PartitionParticles()
   &InitializeParticle(
    Lifetime=2
    Mesh Scale Mode=::NewEnumerator3
    Mesh Uniform Scale=2.5
    Sprite Size Mode=::NewEnumerator3
    Uniform Sprite Size=100
    UsePositionOffset=false
    Write Material Random=true
   )
   &InitialMeshOrientation(Mesh Axis To Orient=<0 0 -1>, Mesh Orientation Mode=::NewEnumerator2, Rotation Coordinate Space=::NewEnumerator0, Use Rotation Vector=false)
   /SetVariables(Particles.LightRadius=120) {isSetParameters=true}
  ]
  particleUpdateScript [
   &ParticleState(
    Kill Particles When Lifetime Has Elapsed=false
    Let Infinitely Lived Particles Die When Emitter Deactivates=true
    Loop Particles Lifetime=true
   )
   &SkeletalMeshLocation(
    Bone / Socket Index=$Particles.PartitionParticles.Partition
    Bone Sampling Mode=::NewEnumerator4
    MaskSpawnGroup=false
    Orientation Sampling=::NewEnumerator1
    Sampled Position Offset=&Lerp_Vector(A=<10 0 0>, B=<-10 0 0>, Alpha=&MakeFloatFromInt(Int=$Particles.PartitionParticles.Partition))
    Skeletal Mesh=$System.Skeletal Mesh
   )
   &UpdateMeshOrientation(
    Facing Direction=&Lerp_Vector(
     A=&Lerp_Vector(A=<1 -1 0.5>, B=<-1 -1 0.5>, Alpha=&MakeFloatFromInt(Int=$Particles.PartitionParticles.Partition))
     B=$System.DiveVelocityScale
     Alpha=&FloatFromCurve(
      FloatCurve=DataInterface() {propertyValues {Curve {keys [
          {arriveTangent=-5.4479, interpMode=RCIM_Cubic, leaveTangent=-5.4479, tangentMode=RCTM_User, time=0, value=0}
          {interpMode=RCIM_Cubic, leaveTangent=1.6667, time=0.2, value=-0.5}
          <at 0.5 1 RCIM_Cubic>
         ]}}}
      CurveIndex=$Emitter.Age
     )
    )
    Facing Interpolation=false
    Orientation Method=::NewEnumerator1
    Override Quat=false
    Side Direction:Vector3f=$Engine.Owner.SystemXAxis
    Stabilization Mode:ENiagara_OrientVectorCount=::NewEnumerator2
    TransformFacing=false
    TransformSide=false
   ) {module=UpdateMeshOrientation001}
   &ScaleMeshSize(Scale Factor=&Multiply_Vector(
     A=<1 1 1>
     B=&VectorFromCurve(
      VectorCurve=DataInterface() {propertyValues {
        XCurve {keys [
          {arriveTangent=-6.6682, interpMode=RCIM_Cubic, leaveTangent=-6.6682, leaveTangentWeight=0.1989, tangentMode=RCTM_User, time=0, value=1}
          {arriveTangentWeight=0.0333, interpMode=RCIM_Cubic, leaveTangentWeight=0.0333, tangentMode=RCTM_User, time=0.1, value=0.4}
          {
           arriveTangent=8.3022
           arriveTangentWeight=0.0333
           interpMode=RCIM_Cubic
           leaveTangent=-8.3169
           leaveTangentWeight=0.0333
           tangentMode=RCTM_Break
           time=0.2
           value=1
          }
          {arriveTangentWeight=0.0333, interpMode=RCIM_Cubic, leaveTangentWeight=0.0333, tangentMode=RCTM_User, time=0.3, value=0.3}
          {arriveTangent=8.5, arriveTangentWeight=0.2233, interpMode=RCIM_Cubic, leaveTangent=8.5, leaveTangentWeight=0.1989, time=0.4, value=1}
          {arriveTangentWeight=0.0333, interpMode=RCIM_Cubic, leaveTangentWeight=0.0333, tangentMode=RCTM_User, time=0.5, value=2}
          {arriveTangent=-12.82, arriveTangentWeight=0.0333, interpMode=RCIM_Cubic, leaveTangent=-12.82, tangentMode=RCTM_User, time=0.6, value=1}
         ]}
        YCurve {keys [
          {arriveTangent=-6.6682, interpMode=RCIM_Cubic, leaveTangent=-6.6682, tangentMode=RCTM_User, time=0, value=1}
          <at 0.1 0.4 RCIM_Cubic>
          {arriveTangent=7, interpMode=RCIM_Cubic, leaveTangent=7, time=0.2, value=1}
          <at 0.3 1.8 RCIM_Cubic>
          {arriveTangent=-8, interpMode=RCIM_Cubic, leaveTangent=-8, time=0.4, value=1}
          <at 0.5 0.2 RCIM_Cubic>
          {arriveTangent=11.695, interpMode=RCIM_Cubic, leaveTangent=11.695, tangentMode=RCTM_User, time=0.6, value=1}
         ]}
        ZCurve {keys [
          {arriveTangent=5.8832, interpMode=RCIM_Cubic, leaveTangent=5.8832, leaveTangentWeight=0.2248, tangentMode=RCTM_User, time=0, value=1}
          {arriveTangentWeight=0.0333, interpMode=RCIM_Cubic, leaveTangentWeight=0.0333, tangentMode=RCTM_User, time=0.1, value=1.6}
          {arriveTangent=-6.5, arriveTangentWeight=0.2081, interpMode=RCIM_Cubic, leaveTangent=-6.5, leaveTangentWeight=0.1872, time=0.2, value=1}
          {arriveTangentWeight=0.0333, interpMode=RCIM_Cubic, leaveTangentWeight=0.0333, tangentMode=RCTM_User, time=0.3, value=0.3}
          {
           arriveTangent=6.625
           arriveTangentWeight=0.0333
           interpMode=RCIM_Cubic
           leaveTangent=-5.8832
           leaveTangentWeight=0.0333
           tangentMode=RCTM_Break
           time=0.4
           value=1
          }
          {arriveTangentWeight=0.0333, interpMode=RCIM_Cubic, leaveTangentWeight=0.0333, tangentMode=RCTM_User, time=0.5, value=0.2}
          {arriveTangent=11.695, arriveTangentWeight=0.0333, interpMode=RCIM_Cubic, leaveTangent=11.695, tangentMode=RCTM_User, time=0.6, value=1}
         ]}
       }}
      CurveIndex=$Emitter.Age
      Scale Curve=&VectorFromFloat(Value=&FloatFromCurve(
        FloatCurve=DataInterface() {propertyValues {Curve {keys [
            {arriveTangent=12.2362, interpMode=RCIM_Cubic, leaveTangent=12.2362, tangentMode=RCTM_User, time=0, value=0}
            {interpMode=RCIM_Cubic, tangentMode=RCTM_User, time=0.2, value=1.4}
            {arriveTangent=-0.25, interpMode=RCIM_Cubic, leaveTangent=-0.25, time=0.4, value=0.7}
            {arriveTangent=0.8571, interpMode=RCIM_Cubic, leaveTangent=0.8571, time=0.6, value=1.3}
            <at 0.75 1 RCIM_Cubic>
           ]}}}
        CurveIndex=$Emitter.Age
       ))
     )
    ))
   /SetVariables(Particles.RandomTime:NiagaraFloat=&Frac_Float(Float=&Add_Float(
      A=$System.NormalizedLoopAge
      B=&Multiply_FloatByInt(Float=0.25, Integer=$Particles.PartitionParticles.Partition)
     ))) {isSetParameters=true}
   &DynamicMaterialParameters(
    Index 0 Param 1=&Lerp_Float(
     A=0
     B=0.4
     Alpha=&FloatFromCurve(
      FloatCurve=DataInterface() {propertyValues {Curve {keys [
          <at 0 0 RCIM_Cubic>
          {arriveTangent=4.1653, interpMode=RCIM_Cubic, leaveTangent=4.1653, time=0.0611, value=0.8}
          {arriveTangent=1.291, interpMode=RCIM_Cubic, leaveTangent=1.291, time=0.1801, value=0.75}
          {arriveTangent=-2.9431, interpMode=RCIM_Cubic, leaveTangent=-2.9431, time=0.216, value=1}
          {arriveTangent=-2.6528, interpMode=RCIM_Cubic, leaveTangent=-2.6528, time=0.316, value=0.35}
          {arriveTangent=-2.6149, interpMode=RCIM_Cubic, leaveTangent=-2.6149, time=0.3516, value=0.6401}
          {interpMode=RCIM_Cubic, tangentMode=RCTM_User, time=0.4498, value=0}
          {arriveTangent=3.486, interpMode=RCIM_Cubic, leaveTangent=3.486, time=0.5, value=0.8}
          {arriveTangent=-0.8651, interpMode=RCIM_Cubic, leaveTangent=-0.8651, time=0.5932, value=0.5}
          {arriveTangent=-4.2832, interpMode=RCIM_Cubic, leaveTangent=-4.2832, time=0.6325, value=0.6853}
          {arriveTangent=-1.4195, interpMode=RCIM_Cubic, leaveTangent=-1.4195, time=0.71, value=0}
          {arriveTangent=2.1663, interpMode=RCIM_Cubic, leaveTangent=2.1663, time=0.7983, value=0.45}
          {arriveTangent=5.4095, interpMode=RCIM_Cubic, leaveTangent=5.4095, time=0.8715, value=0.35}
          {arriveTangent=-2.7248, interpMode=RCIM_Cubic, leaveTangent=-2.7248, time=0.9, value=1}
          <at 1 0 RCIM_Cubic>
         ]}}}
      CurveIndex=$Particles.RandomTime
      Scale Curve=&Lerp_Float(A=0.4, B=1, Alpha=$System.FN_Integrator001.Value)
     )
    )
    Module.Param0WriteEnabled=true
    Module.Param1WriteEnabled=false
    Module.Param2WriteEnabled=false
    Module.Param3WriteEnabled=false
   )
   /SetVariables(Particles.DynamicMaterialParameter1:Vector4f=&MakeFloat4FromVectorAndFloat(
     "Vector (RGB)"=$System.DiveVelocityScale
     "Float (Alpha)"=&FloatFromCurve(
      FloatCurve=DataInterface() {propertyValues {Curve {keys [<at 0.5 0 RCIM_Cubic>, {interpMode=RCIM_Cubic, tangentMode=RCTM_User, time=0.7, value=2}, <at 0.9 1 RCIM_Cubic>]}}}
      CurveIndex=$Emitter.Age
      Scale Curve=&Lerp_Float(A=8, B=24, Alpha=$System.FN_Integrator001.Value)
     )
    )) {isSetParameters=true}
   &DynamicMaterialParameters(
    Index 0 Param 1=1
    Module.Param0WriteEnabled=true
    Module.Param1WriteEnabled=false
    Module.Param2WriteEnabled=false
    Module.Param3WriteEnabled=false
   ) {enabled=false, module=DynamicMaterialParameters001}
   &Color(Color=&LerpLinearColors(
     StartColor=$User.ColorA
     EndColor=$User.ColorB
     LerpFactor=&Lerp_Float(
      A=&FloatFromCurve(
       FloatCurve=DataInterface() {propertyValues {Curve {keys [<at 0 0 RCIM_Cubic>, <at 0.1 1 RCIM_Cubic>, {interpMode=RCIM_Cubic, tangentMode=RCTM_User, time=0.6, value=0}]}}}
       CurveIndex=$Emitter.Age
      )
      B=0.33
      Alpha=&FloatFromCurve(
       FloatCurve=DataInterface() {propertyValues {Curve {keys [
           <at 0 0 RCIM_Cubic>
           {arriveTangent=3.8876, interpMode=RCIM_Cubic, leaveTangent=3.8876, time=0.0611, value=0.75}
           {arriveTangent=1.6137, interpMode=RCIM_Cubic, leaveTangent=1.6137, time=0.1801, value=0.7}
           {arriveTangent=-2.5752, interpMode=RCIM_Cubic, leaveTangent=-2.5752, time=0.216, value=1}
           {arriveTangent=-2.6528, interpMode=RCIM_Cubic, leaveTangent=-2.6528, time=0.316, value=0.35}
           {arriveTangent=-2.6149, interpMode=RCIM_Cubic, leaveTangent=-2.6149, time=0.3516, value=0.6401}
           {interpMode=RCIM_Cubic, tangentMode=RCTM_User, time=0.4498, value=0}
           {arriveTangent=3.486, interpMode=RCIM_Cubic, leaveTangent=3.486, time=0.5, value=1}
           {arriveTangent=-2.3742, interpMode=RCIM_Cubic, leaveTangent=-2.3742, time=0.5932, value=0.5}
           {arriveTangent=-4.2832, interpMode=RCIM_Cubic, leaveTangent=-4.2832, time=0.6325, value=0.6853}
           {arriveTangent=-0.2132, interpMode=RCIM_Cubic, leaveTangent=-0.2132, time=0.71, value=0}
           {arriveTangent=2.1663, interpMode=RCIM_Cubic, leaveTangent=2.1663, time=0.7983, value=0.65}
           {arriveTangent=3.4424, interpMode=RCIM_Cubic, leaveTangent=3.4424, time=0.8715, value=0.35}
           {arriveTangent=-2.7248, interpMode=RCIM_Cubic, leaveTangent=-2.7248, time=0.9, value=1}
           <at 1 0 RCIM_Cubic>
          ]}}}
       CurveIndex=$Particles.RandomTime
       Scale Curve=&Lerp_Float(A=0.4, B=1, Alpha=$System.FN_Integrator001.Value)
      )
     )
    ))
   &FN_Sprite_Attributes(
    Camera Offset=20
    Color=&MultiplyLinearColor(
     Linear Color=&LerpLinearColors(
      StartColor=$User.ColorA
      EndColor=$User.ColorB
      LerpFactor=&Lerp_Float(
       A=0.9
       B=1
       Alpha=&FloatFromCurve(
        FloatCurve=DataInterface() {propertyValues {Curve {keys [
            <at 0 0 RCIM_Cubic>
            {arriveTangent=1.304, interpMode=RCIM_Cubic, leaveTangent=1.304, time=0.1, value=0.75}
            {arriveTangent=-1.2483, interpMode=RCIM_Cubic, leaveTangent=-1.2483, time=0.1681, value=0.2192}
            {arriveTangent=-0.9034, interpMode=RCIM_Cubic, leaveTangent=-0.9034, time=0.216, value=1}
            {arriveTangent=-0.1111, interpMode=RCIM_Cubic, leaveTangent=-0.1111, time=0.3, value=0.1}
            {arriveTangent=1.8356, interpMode=RCIM_Cubic, leaveTangent=1.8356, time=0.3516, value=0.6401}
            {arriveTangent=2.4257, interpMode=RCIM_Cubic, leaveTangent=2.4257, time=0.4219, value=0.05}
            {arriveTangent=0.2031, interpMode=RCIM_Cubic, leaveTangent=0.2031, time=0.5, value=1}
            {arriveTangent=-2.3742, interpMode=RCIM_Cubic, leaveTangent=-2.3742, time=0.5703, value=0.3538}
            {arriveTangent=-1.957, interpMode=RCIM_Cubic, leaveTangent=-1.957, time=0.6325, value=0.6853}
            {arriveTangent=-0.2171, interpMode=RCIM_Cubic, leaveTangent=-0.2171, time=0.7, value=0}
            {arriveTangent=1.2231, interpMode=RCIM_Cubic, leaveTangent=1.2231, time=0.7714, value=0.6552}
            {arriveTangent=1.1257, interpMode=RCIM_Cubic, leaveTangent=1.1257, time=0.8336, value=0.2634}
            {arriveTangent=-1.5833, interpMode=RCIM_Cubic, leaveTangent=-1.5833, time=0.9, value=1}
            <at 1 0 RCIM_Cubic>
           ]}}}
        CurveIndex=$System.NormalizedLoopAge
        Scale Curve=&Lerp_Float(A=1, B=2, Alpha=$System.FN_Integrator001.Value)
       )
      )
     )
     Scale Factor=&MakeFloat4FromVectorAndFloat(
      "Vector (RGB)"=<1 1 1>
      "Float (Alpha)"=&FloatFromCurve(
       FloatCurve=DataInterface() {propertyValues {Curve {keys [<at 0.5 0 RCIM_Cubic>, {interpMode=RCIM_Cubic, tangentMode=RCTM_User, time=1, value=1}]}}}
       CurveIndex=$Emitter.Age
      )
     )
    )
    Create Camera Offset Attribute=true
    Create Color Attribute=true
    Write to Module Namespace=true
   )
  ]
  rendererClasses [
   &NiagaraLightRendererProperties
   &NiagaraSpriteRendererProperties
   &NiagaraMeshRendererProperties
  ]
  renderers [
   {rendererClass=&NiagaraLightRendererProperties, rendererIndex=0}
   {rendererClass=&NiagaraSpriteRendererProperties, rendererIndex=1}
   {rendererClass=&NiagaraMeshRendererProperties, rendererIndex=2}
   {rendererClass=&NiagaraMeshRendererProperties, rendererIndex=3}
  ]
 }
 Cube_Trails {
  emitterUpdateScript [
   &EmitterState(Delay First Loop Only=true, Life Cycle Mode=::NewEnumerator1, Loop Delay=0.5, Loop Duration=1, UseLoopDelay=true)
   &SpawnRate(SpawnRate=&Lerp_Float(A=40, B=40, Alpha=$System.FN_Integrator001.Value), Use Spawn Probability=true)
  ]
  particleSpawnScript [
   &PartitionParticles()
   &InitializeParticle(
    Color=<rgba 2 2 2 1>
    Lifetime Max=0.7
    Lifetime Min=0.6
    Lifetime Mode=::NewEnumerator1
    Mesh Scale=&MakeVector(
     X=&RandomRangeFloat(Minimum=7.5, Maximum=10)
     Y=&RandomRangeFloat(Minimum=1.5, Maximum=3)
     Z=&RandomRangeFloat(Minimum=1.5, Maximum=5)
    )
    Mesh Scale Mode=::NewEnumerator4
    UsePositionOffset=true
    Write Material Random=true
   )
   &InitialMeshOrientation(
    Mesh Orientation Mode=::NewEnumerator2
    Rotation=&RandomRangeVector(Minimum=<0 0 0>, Maximum=<0 0 0>)
    Rotation Coordinate Space=::NewEnumerator0
    Use Rotation Vector=false
   )
  ]
  particleUpdateScript [
   &ParticleState()
   &UpdateMeshOrientation(
    Facing Direction=$System.DiveVelocityScale
    Facing Interpolation=false
    Orientation Method=::NewEnumerator1
    Override Quat=false
    TransformFacing=false
    TransformUp=false
    Up Direction:Vector3f=$Engine.Owner.SystemYAxis
   ) {module=UpdateMeshOrientation001}
   &SkeletalMeshLocation(
    Bone / Socket Index=$Particles.PartitionParticles.Partition
    Bone Sampling Mode=::NewEnumerator4
    MaskSpawnGroup=false
    Orientation Sampling=::NewEnumerator1
    Sampled Position Offset=&Lerp_Vector(A=<10 0 0>, B=<-10 0 0>, Alpha=&MakeFloatFromInt(Int=$Particles.PartitionParticles.Partition))
    Skeletal Mesh=$System.Skeletal Mesh
   )
   &ShapeLocation(
    Additional Quaternion Rotation=true
    Additional Yaw / Pitch / Roll=<0 0.25 0>
    Box / Plane Mode:ENiagara_BoxPlaneMode=::NewEnumerator1
    MaskSpawnGroup=false
    Override Seed=false
    Plane Size:Vector2f=<12 12>
    Rotation Mode=::NewEnumerator3
    Rotation Quaternion=$Output.UpdateMeshOrientation001.MeshOrientation
    Shape Origin=&ConvertPositionToVector(Input Position=$Particles.Position, Convert To Absolute Large World Vector=false)
    Shape Primitive=::NewEnumerator2
    UseOwnerScale=true
   )
   /SetVariables(Particles.Position=&Add_Vector(
     A=&ConvertPositionToVector(Input Position=$Particles.Position, Convert To Absolute Large World Vector=false)
     B=&Multiply_VectorByFloat(
      Vector=$System.DiveVelocityScale
      Float=&Add_Float(
       A=&FloatFromCurve(
        FloatCurve=DataInterface() {propertyValues {Curve {keys [
            {arriveTangent=0.3747, interpMode=RCIM_Cubic, leaveTangent=0.3747, tangentMode=RCTM_User, time=0, value=0}
            {arriveTangent=1.721, interpMode=RCIM_Cubic, leaveTangent=1.721, tangentMode=RCTM_User, time=1, value=1}
           ]}}}
        Scale Curve=&Multiply_Float(A=$System.DiveOffset, B=&RandomRangeFloat(Minimum=0.5, Maximum=1))
       )
       B=-11
      )
     )
    )) {isSetParameters=true, module=SetVariables_#8}
   &ScaleMeshSize(Scale Factor=&Multiply_VectorByFloat(
     Vector=&Lerp_Vector(
      A=<0.5 1 1>
      B=<1 1 1>
      Alpha=&FloatFromCurve(FloatCurve=DataInterface() {propertyValues {Curve {keys [
           {arriveTangent=0.9666, interpMode=RCIM_Cubic, time=0.2, value=0}
           {interpMode=RCIM_Cubic, tangentMode=RCTM_User, time=0.5, value=1}
           {interpMode=RCIM_Cubic, tangentMode=RCTM_User, time=0.6, value=1}
           <at 0.9 0>
          ]}}})
     )
     Float=&FloatFromCurve(FloatCurve=DataInterface() {propertyValues {Curve {keys [
          {arriveTangent=3.3011, interpMode=RCIM_Cubic, leaveTangent=3.3011, tangentMode=RCTM_User, time=0, value=0.25}
          {interpMode=RCIM_Cubic, tangentMode=RCTM_User, time=0.15, value=1}
          {interpMode=RCIM_Cubic, tangentMode=RCTM_Break, time=0.4, value=1}
          {arriveTangent=-1.8681, interpMode=RCIM_Cubic, leaveTangent=-1.8681, tangentMode=RCTM_User, time=1, value=0.25}
         ]}}})
    ))
   &DynamicMaterialParameters(
    Index 0 Param 1=&FloatFromCurve(
     FloatCurve=DataInterface() {propertyValues {Curve {keys [
         <at 0 0 RCIM_Cubic>
         {arriveTangent=1, interpMode=RCIM_Cubic, leaveTangent=1, time=0.6, value=0.375}
         {arriveTangent=1.6037, interpMode=RCIM_Cubic, leaveTangent=1.6037, tangentMode=RCTM_User, time=1, value=1}
        ]}}}
     Scale Curve=$System.Cube Emissive
    )
    Module.Param0WriteEnabled=true
    Module.Param1WriteEnabled=false
    Module.Param2WriteEnabled=false
    Module.Param3WriteEnabled=false
   )
   &Color(Color=&LerpLinearColors(
     StartColor=$User.ColorA
     EndColor=$User.ColorB
     LerpFactor=&FloatFromCurve(FloatCurve=DataInterface() {propertyValues {Curve {keys [{interpMode=RCIM_Cubic, tangentMode=RCTM_User, time=0.4, value=0}, {interpMode=RCIM_Cubic, leaveTangent=5.1459, time=0.75, value=1}]}}})
    ))
  ]
  rendererClasses [&NiagaraMeshRendererProperties]
  renderers [{rendererClass=&NiagaraMeshRendererProperties, rendererIndex=0}]
 }
 Cube_Trails_Small {
  simTarget=GPUComputeSim
  emitterUpdateScript [
   &EmitterState(Delay First Loop Only=true, Life Cycle Mode=::NewEnumerator1, Loop Delay=0.5, Loop Duration=1, UseLoopDelay=true)
   &SpawnRate(SpawnRate=&Lerp_Float(A=400, B=400, Alpha=$System.FN_Integrator001.Value), Use Spawn Probability=true)
  ]
  particleSpawnScript [
   &PartitionParticles() {module=PartitionParticles001}
   &InitializeParticle(
    Color=<rgba 2 2 2 1>
    Lifetime Max=0.6
    Lifetime Min=0.5
    Lifetime Mode=::NewEnumerator1
    Mesh Scale=&MakeVector(
     X=&RandomRangeFloat(Minimum=1.5, Maximum=4)
     Y=&RandomRangeFloat(Minimum=0.75, Maximum=1.5)
     Z=&RandomRangeFloat(Minimum=0.75, Maximum=1.5)
    )
    Mesh Scale Mode=::NewEnumerator4
    UsePositionOffset=false
    Write Material Random=true
   )
   &InitialMeshOrientation(Mesh Orientation Mode=::NewEnumerator2, Rotation Coordinate Space=::NewEnumerator2, Use Rotation Vector=false)
   &ShapeLocation(Box Size=<20 20 20>, MaskSpawnGroup=false, Override Seed=false, Shape Primitive=::NewEnumerator2, Surface Only=false, UseOwnerScale=true)
  ]
  particleUpdateScript [
   &ParticleState()
   &UpdateMeshOrientation(
    Facing Direction=$System.DiveVelocityScale
    Facing Interpolation=false
    Orientation Method=::NewEnumerator1
    Override Quat=false
    Side Direction:Vector3f=$Engine.Owner.SystemXAxis
    Stabilization Mode:ENiagara_OrientVectorCount=::NewEnumerator2
    TransformFacing=false
    TransformSide=false
   )
   &SkeletalMeshLocation(
    Bone / Socket Index=$Particles.PartitionParticles001.Partition
    Bone Sampling Mode=::NewEnumerator4
    MaskSpawnGroup=false
    Orientation Sampling=::NewEnumerator1
    Sampled Position Offset=&Lerp_Vector(A=<10 0 0>, B=<-10 0 0>, Alpha=&MakeFloatFromInt(Int=$Particles.PartitionParticles001.Partition))
    Skeletal Mesh=$System.Skeletal Mesh
   )
   /SetVariables(Particles.Position=&Add_Vector(
     A=&ConvertPositionToVector(Input Position=$Particles.Position, Convert To Absolute Large World Vector=false)
     B=&Add_Vector(
      A=&Multiply_VectorByFloat(
       Vector=$System.DiveVelocityScale
       Float=&Add_Float(
        A=&FloatFromCurve(
         FloatCurve=DataInterface() {propertyValues {Curve {keys [
             {arriveTangent=0.3747, interpMode=RCIM_Cubic, leaveTangent=0.3747, tangentMode=RCTM_User, time=0, value=0}
             {arriveTangent=1.721, interpMode=RCIM_Cubic, leaveTangent=1.721, tangentMode=RCTM_User, time=1, value=1}
            ]}}}
         Scale Curve=&Multiply_Float(A=$System.DiveOffset, B=&RandomRangeFloat(Minimum=0.05, Maximum=0.4))
        )
        B=-11
       )
      )
      B=&RotateVectorByQuaternion(
       Vector To Rotate=&Multiply_Vector(
        A=&MakeCustomVectorFromBool(
         Bool=&RandomBool()
         True Vector=&MakeCustomVectorFromBool(
          Bool=&RandomBool()
          True Vector=&RandomRangeVector(Minimum=<0 1 -1>, Maximum=<0 1 1>)
          False Vector=&RandomRangeVector(Minimum=<0 -1 -1>, Maximum=<0 -1 1>)
         )
         False Vector=&MakeCustomVectorFromBool(
          Bool=&RandomBool()
          True Vector=&RandomRangeVector(Minimum=<0 -1 1>, Maximum=<0 1 1>)
          False Vector=&RandomRangeVector(Minimum=<0 -1 -1>, Maximum=<0 1 -1>)
         )
        )
        B=<0 11 11>
       )
       Quaternion=$Output.UpdateMeshOrientation.MeshOrientation
      )
     )
    )) {isSetParameters=true}
   &ScaleMeshSize(Scale Factor=&Multiply_VectorByFloat(
     Vector=&Lerp_Vector(
      A=<0.5 1 1>
      B=<1 1 1>
      Alpha=&FloatFromCurve(FloatCurve=DataInterface() {propertyValues {Curve {keys [
           {arriveTangent=0.9666, interpMode=RCIM_Cubic, time=0.2, value=0}
           {interpMode=RCIM_Cubic, tangentMode=RCTM_User, time=0.6, value=1}
           {interpMode=RCIM_Cubic, tangentMode=RCTM_User, time=0.7, value=1}
           <at 0.9 0>
          ]}}})
     )
     Float=&FloatFromCurve(FloatCurve=DataInterface() {propertyValues {Curve {keys [
          {arriveTangent=3.3011, interpMode=RCIM_Cubic, leaveTangent=3.3011, tangentMode=RCTM_User, time=0, value=0.25}
          {interpMode=RCIM_Cubic, tangentMode=RCTM_User, time=0.15, value=1}
          {interpMode=RCIM_Cubic, tangentMode=RCTM_Break, time=0.8, value=1}
          {arriveTangent=-3.8115, interpMode=RCIM_Constant, leaveTangent=-3.8115, time=1, value=0.25}
         ]}}})
    ))
   &DynamicMaterialParameters(
    Index 0 Param 1=&FloatFromCurve(
     FloatCurve=DataInterface() {propertyValues {Curve {keys [
         <at 0 0 RCIM_Cubic>
         {arriveTangent=1, interpMode=RCIM_Cubic, leaveTangent=1, time=0.6, value=0.375}
         {arriveTangent=1.6037, interpMode=RCIM_Cubic, leaveTangent=1.6037, tangentMode=RCTM_User, time=1, value=1}
        ]}}}
     Scale Curve=$System.Cube Emissive
    )
    Module.Param0WriteEnabled=true
    Module.Param1WriteEnabled=false
    Module.Param2WriteEnabled=false
    Module.Param3WriteEnabled=false
   )
   &Color(Color=&LerpLinearColors(
     StartColor=$User.ColorA
     EndColor=$User.ColorB
     LerpFactor=&FloatFromCurve(FloatCurve=DataInterface() {propertyValues {Curve {keys [{interpMode=RCIM_Cubic, tangentMode=RCTM_User, time=0.4, value=0}, {interpMode=RCIM_Cubic, leaveTangent=5.1459, time=0.75, value=1}]}}})
    ))
  ]
  rendererClasses [&NiagaraMeshRendererProperties]
  renderers [{rendererClass=&NiagaraMeshRendererProperties, rendererIndex=0}]
 }
 Cubes_Inside {
  emitterUpdateScript [
   &EmitterState(Delay First Loop Only=true, Life Cycle Mode=::NewEnumerator1, Loop Delay=0.5, Loop Duration=1, UseLoopDelay=true)
   &SpawnRate(SpawnRate=&Lerp_Float(A=40, B=50, Alpha=$System.FN_Integrator001.Value), Use Spawn Probability=true)
  ]
  particleSpawnScript [
   &PartitionParticles()
   &InitializeParticle(
    Color=<rgba 2 2 2 1>
    Lifetime Max=1.25
    Lifetime Mode=::NewEnumerator1
    Mesh Scale Mode=::NewEnumerator1
    Mesh Uniform Scale Min=1.5
    UsePositionOffset=true
    Write Material Random=true
   )
   /SetVariables(Particles.Lifetime=&Multiply_Float(A=$Particles.Lifetime, B=&Lerp_Float(A=1, B=0.75, Alpha=$System.FN_Integrator001.Value))) {isSetParameters=true}
   &InitialMeshOrientation(
    Mesh Orientation Mode=::NewEnumerator2
    Rotation=&RandomRangeVector(Minimum=<0 0 0>, Maximum=<0 0 0>)
    Rotation Coordinate Space=::NewEnumerator0
    Use Rotation Vector=false
   )
  ]
  particleUpdateScript [
   &ParticleState()
   &UpdateMeshOrientation(
    Facing Direction=$System.DiveVelocityScale
    Facing Interpolation=false
    Orientation Method=::NewEnumerator1
    Override Quat=false
    TransformFacing=false
    TransformUp=false
    Up Direction:Vector3f=$Engine.Owner.SystemYAxis
   ) {module=UpdateMeshOrientation001}
   &SkeletalMeshLocation(
    Bone / Socket Index=$Particles.PartitionParticles.Partition
    Bone Sampling Mode=::NewEnumerator4
    MaskSpawnGroup=false
    Orientation Sampling=::NewEnumerator1
    Sampled Position Offset=&Lerp_Vector(A=<10 0 0>, B=<-10 0 0>, Alpha=&MakeFloatFromInt(Int=$Particles.PartitionParticles.Partition))
    Skeletal Mesh=$System.Skeletal Mesh
   )
   &ShapeLocation(
    Additional Quaternion Rotation=true
    Additional Yaw / Pitch / Roll=<0 0.25 0>
    Box Size=<20 20 0>
    MaskSpawnGroup=false
    Override Seed=false
    Rotation Mode=::NewEnumerator3
    Rotation Quaternion=$Output.UpdateMeshOrientation001.MeshOrientation
    Shape Origin=&ConvertPositionToVector(Input Position=$Particles.Position, Convert To Absolute Large World Vector=false)
    Shape Primitive=::NewEnumerator2
    Surface Only=true
    UseOwnerScale=true
   ) {module=ShapeLocation001}
   /SetVariables(Particles.Position=&Add_Vector(
     A=&ConvertPositionToVector(Input Position=$Particles.Position, Convert To Absolute Large World Vector=false)
     B=&Multiply_VectorByFloat(
      Vector=$System.DiveVelocityScale
      Float=&Add_Float(
       A=&FloatFromCurve(
        FloatCurve=DataInterface() {propertyValues {Curve {keys [
            {interpMode=RCIM_Cubic, tangentMode=RCTM_User, time=0, value=0}
            {arriveTangent=1.3649, interpMode=RCIM_Cubic, leaveTangent=1.3649, tangentMode=RCTM_User, time=1, value=1}
           ]}}}
        Scale Curve=-24
       )
       B=10
      )
     )
    )) {isSetParameters=true, module=SetVariables_#8}
   &ShapeLocation(
    Box / Plane Mode:ENiagara_BoxPlaneMode=::NewEnumerator1
    MaskSpawnGroup=false
    Override Seed=false
    Plane Size:Vector2f=<10 10>
    Shape Origin=$Particles.SkeletalMeshLocation.SampledPosition
    Shape Primitive=::NewEnumerator2
    UseOwnerScale=true
   ) {enabled=false}
   &ScaleMeshSize(Scale Factor=&Multiply_VectorByFloat(
     Vector=<1 1 1>
     Float=&FloatFromCurve(FloatCurve=DataInterface() {propertyValues {Curve {keys [
          {arriveTangent=3.3011, interpMode=RCIM_Cubic, leaveTangent=3.3011, tangentMode=RCTM_User, time=0, value=0.25}
          {interpMode=RCIM_Cubic, tangentMode=RCTM_User, time=0.15, value=1}
          {interpMode=RCIM_Cubic, tangentMode=RCTM_Break, time=0.9, value=1}
          {arriveTangent=-1.8681, interpMode=RCIM_Cubic, leaveTangent=-1.8681, tangentMode=RCTM_User, time=1, value=0.25}
         ]}}})
    ))
   &DynamicMaterialParameters(
    Index 0 Param 1=&FloatFromCurve(
     FloatCurve=DataInterface() {propertyValues {Curve {keys [
         {interpMode=RCIM_Cubic, tangentMode=RCTM_User, time=0, value=0}
         {interpMode=RCIM_Cubic, tangentMode=RCTM_User, time=0.5, value=1}
         {tangentMode=RCTM_User, time=1, value=0}
        ]}}}
     Scale Curve=0.25
    )
    Module.Param0WriteEnabled=true
    Module.Param1WriteEnabled=false
    Module.Param2WriteEnabled=false
    Module.Param3WriteEnabled=false
   )
   &Color(Color=&LerpLinearColors(
     StartColor=$User.ColorA
     EndColor=$User.ColorB
     LerpFactor=&FloatFromCurve(
      FloatCurve=DataInterface() {propertyValues {Curve {keys [
          {interpMode=RCIM_Cubic, tangentMode=RCTM_User, time=0, value=0}
          {interpMode=RCIM_Cubic, tangentMode=RCTM_User, time=0.5, value=1}
          {tangentMode=RCTM_User, time=1, value=0}
         ]}}}
      Scale Curve=0
     )
    ))
  ]
  rendererClasses [&NiagaraMeshRendererProperties]
  renderers [{rendererClass=&NiagaraMeshRendererProperties, rendererIndex=0}]
 }
 Energy {
  emitterUpdateScript [
   &EmitterState(Life Cycle Mode=::NewEnumerator1, Loop Behavior=::NewEnumerator1, Loop Delay=0.5, Loop Duration Mode=::NewEnumerator1, UseLoopDelay=true)
   &SpawnBurst_Instantaneous(Spawn Count=2, Use Spawn Probability=false, UseLoopCountLimit=false)
  ]
  particleSpawnScript [
   &PartitionParticles()
   &InitializeParticle(
    Lifetime=2
    Mesh Scale Mode=::NewEnumerator3
    Sprite Size Mode=::NewEnumerator3
    Uniform Sprite Size=50
    UsePositionOffset=false
    Write Material Random=true
   )
   &InitialMeshOrientation(
    Mesh Axis To Orient=<0 0 -1>
    Mesh Orientation Mode=::NewEnumerator2
    Rotation=&Lerp_Vector(A=<0 0 0>, B=<0.5 0 0>, Alpha=&MakeFloatFromInt(Int=$Particles.PartitionParticles.Partition))
    Rotation Coordinate Space=::NewEnumerator0
    Use Rotation Vector=true
   )
  ]
  particleUpdateScript [
   &ParticleState(
    Kill Particles When Lifetime Has Elapsed=false
    Let Infinitely Lived Particles Die When Emitter Deactivates=true
    Loop Particles Lifetime=true
   )
   &SkeletalMeshLocation(
    Bone / Socket Index=$Particles.PartitionParticles.Partition
    Bone Sampling Mode=::NewEnumerator4
    MaskSpawnGroup=false
    Orientation Sampling=::NewEnumerator1
    Sampled Position Offset=&Lerp_Vector(A=<10 0 0>, B=<-10 0 0>, Alpha=&MakeFloatFromInt(Int=$Particles.PartitionParticles.Partition))
    Skeletal Mesh=$System.Skeletal Mesh
   )
   /SetVariables(Particles.Position=&Add_Vector(
     A=&ConvertPositionToVector(Input Position=$Particles.Position, Convert To Absolute Large World Vector=false)
     B=&Add_Vector(A=&Multiply_VectorByFloat(Vector=$System.DiveVelocityScale, Float=10), B=<0 0 0>)
    )) {isSetParameters=true}
   &UpdateMeshOrientation(
    Facing Direction=$System.DiveVelocityScale
    Facing Interpolation=false
    Orientation Method=::NewEnumerator1
    Override Quat=false
    Side Direction:Vector3f=$Engine.Owner.SystemXAxis
    Stabilization Mode:ENiagara_OrientVectorCount=::NewEnumerator2
    TransformFacing=false
    TransformSide=false
   ) {module=UpdateMeshOrientation001}
   &UpdateMeshOrientation(Rotation Rate=-1, Rotation Vector=<0 0 1>)
   &CameraOffset(Camera Offset Amount=8)
   &ScaleMeshSize(Scale Factor=&Multiply_Vector(
     A=&Lerp_Vector(A=<1 1 1.25>, B=<1 1 1.5>, Alpha=$System.FN_Integrator001.Value)
     B=&MakeVector(
      X=1
      Y=1
      Z=&FloatFromCurve(
       FloatCurve=DataInterface() {propertyValues {Curve {keys [<at 0.5 0 RCIM_Cubic>, {arriveTangent=4.1531, interpMode=RCIM_Cubic, leaveTangent=4.1531, tangentMode=RCTM_User, time=1, value=1}]}}}
       CurveIndex=$Emitter.Age
      )
     )
    ))
   &ScaleColor(
    Scale Alpha=&FloatFromCurve(
     FloatCurve=DataInterface() {propertyValues {Curve {keys [<at 0.5 0 RCIM_Cubic>, {arriveTangent=4.1531, interpMode=RCIM_Cubic, leaveTangent=4.1531, tangentMode=RCTM_User, time=1, value=1}]}}}
     CurveIndex=$Emitter.Age
     Scale Curve=0.6
    )
    Scale RGB=&Multiply_VectorByFloat(Vector=&MakeVectorFromLinearColorRGB(Color=$User.ColorA), Float=0.8)
    ScaleA=true
    ScaleRGB=true
   ) {module=ScaleColor001}
   /SetVariables(Particles.Color_Yellow:LinearColor=&MakeLinearColorFromVectorAndFloat(
     "Vector (RGB)"=&Multiply_VectorByFloat(Vector=&MakeVectorFromLinearColorRGB(Color=$User.ColorB), Float=2)
     "Float (Alpha)"=&FloatFromCurve(
      FloatCurve=DataInterface() {propertyValues {Curve {keys [<at 0.5 0 RCIM_Cubic>, {arriveTangent=4.1531, interpMode=RCIM_Cubic, leaveTangent=4.1531, tangentMode=RCTM_User, time=1, value=1}]}}}
      CurveIndex=$Emitter.Age
     )
    )) {isSetParameters=true}
   /SetVariables(Particles.Scale_Yellow:Vector3f=&Multiply_Vector(A=$Particles.Scale, B=<0.7 0.7 0.5>)) {isSetParameters=true}
   &ScaleColor(
    Scale Alpha=&FloatFromCurve(
     FloatCurve=DataInterface() {propertyValues {Curve {keys [<at 0.5 0 RCIM_Cubic>, {arriveTangent=4.1531, interpMode=RCIM_Cubic, leaveTangent=4.1531, tangentMode=RCTM_User, time=1, value=1}]}}}
     CurveIndex=$Emitter.Age
    )
    Scale RGB=&MakeVectorFromLinearColorRGB(Color=&LerpLinearColors(
      StartColor=$User.ColorA
      EndColor=$User.ColorB
      LerpFactor=&Lerp_Float(
       A=0
       B=1
       Alpha=&FloatFromCurve(
        FloatCurve=DataInterface() {propertyValues {Curve {keys [
            <at 0 0 RCIM_Cubic>
            {arriveTangent=1.304, interpMode=RCIM_Cubic, leaveTangent=1.304, time=0.1, value=0.75}
            {arriveTangent=-1.2483, interpMode=RCIM_Cubic, leaveTangent=-1.2483, time=0.1681, value=0.2192}
            {arriveTangent=-0.9034, interpMode=RCIM_Cubic, leaveTangent=-0.9034, time=0.216, value=1}
            {arriveTangent=-0.1111, interpMode=RCIM_Cubic, leaveTangent=-0.1111, time=0.3, value=0.1}
            {arriveTangent=1.8356, interpMode=RCIM_Cubic, leaveTangent=1.8356, time=0.3516, value=0.6401}
            {arriveTangent=2.4257, interpMode=RCIM_Cubic, leaveTangent=2.4257, time=0.4219, value=0.05}
            {arriveTangent=0.2031, interpMode=RCIM_Cubic, leaveTangent=0.2031, time=0.5, value=1}
            {arriveTangent=-2.3742, interpMode=RCIM_Cubic, leaveTangent=-2.3742, time=0.5703, value=0.3538}
            {arriveTangent=-1.957, interpMode=RCIM_Cubic, leaveTangent=-1.957, time=0.6325, value=0.6853}
            {arriveTangent=-0.2171, interpMode=RCIM_Cubic, leaveTangent=-0.2171, time=0.7, value=0}
            {arriveTangent=1.2231, interpMode=RCIM_Cubic, leaveTangent=1.2231, time=0.7714, value=0.6552}
            {arriveTangent=1.1257, interpMode=RCIM_Cubic, leaveTangent=1.1257, time=0.8336, value=0.2634}
            {arriveTangent=-1.5833, interpMode=RCIM_Cubic, leaveTangent=-1.5833, time=0.9, value=1}
            <at 1 0 RCIM_Cubic>
           ]}}}
        CurveIndex=$System.NormalizedLoopAge
        Scale Curve=&Lerp_Float(A=0.1, B=0.4, Alpha=$System.FN_Integrator001.Value)
       )
      )
     ))
    ScaleA=true
    ScaleRGB=true
   ) {enabled=false}
   &InterpolateOverTime(
    Rate Of Change=10
    Target Vector Value:Vector3f=&ConvertPositionToVector(Input Position=$Particles.Position, Convert To Absolute Large World Vector=false)
    Variable Type=::NewEnumerator2
   ) {enabled=false}
   /SetVariables(Particles.InterpolateDynamic:Vector4f=&MakeFloat4FromVectorAndFloat("Vector (RGB)"=$Output.InterpolateOverTime.Moving Average, "Float (Alpha)"=1)) {enabled=false, isSetParameters=true}
   &DynamicMaterialParameters(
    Index 0 Param 2=&RandomRangeFloat(Minimum=0, Maximum=1.5)
    Index 2 Param 3=0.1
    Index 2 Param 4=1.75
    Module.Param0WriteEnabled=true
    Module.Param0WriteEnabled2=false
    Module.Param1WriteEnabled=true
    Module.Param1WriteEnabled2=false
    Module.Param2WriteEnabled=true
    Module.Param2WriteEnabled2=true
    Module.Param3WriteEnabled=true
    Module.Param3WriteEnabled2=true
    Write Parameter Index 2=true
   )
  ]
  rendererClasses [&NiagaraMeshRendererProperties]
  renderers [
   {rendererClass=&NiagaraMeshRendererProperties, rendererIndex=0}
   {rendererClass=&NiagaraMeshRendererProperties, rendererIndex=1}
  ]
 }
}
summary {
 systemName=NS_Contrail_TheOrigin
 emitters [
  {emitterName=Cube_Trails, rendererClasses [&NiagaraMeshRendererProperties]}
  {
   emitterName=Block_Main_Light
   rendererClasses [
    &NiagaraLightRendererProperties
    &NiagaraSpriteRendererProperties
    &NiagaraMeshRendererProperties
   ]
  }
  {emitterName=Cube_Trails_Small, simTarget=GPUComputeSim, rendererClasses [&NiagaraMeshRendererProperties]}
  {emitterName=Energy, rendererClasses [&NiagaraMeshRendererProperties]}
  {emitterName=Cubes_Inside, rendererClasses [&NiagaraMeshRendererProperties]}
 ]
 userVariables [
  User.bIsFrontEnd {type=NiagaraBool}
  User.bIsFrontEndPreview {type=NiagaraBool, defaultValue=true}
  User.ColorA {type=LinearColor, description="", defaultValue {struct=LinearColor, value=<rgba 0.4938 0 2 1>}}
  User.ColorB {type=LinearColor, description="", defaultValue {struct=LinearColor, value=<rgba 1.25 0.3125 0 1>}}
  User.ParaGlideLeanAlpha {type=NiagaraFloat, defaultValue=0, description=""}
 ]
}
